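# Physics asset for the Virtuous pickaxe, a two-handed melee cosmetic, giving physics bodies to its left and right feather bone chains.
# A capsule on the handle bone anchors two mirrored four-bone chains, each using spheres near the root and capsules toward the tip.
# Angular constraints link each bone to its parent, with swing and twist limits generally loosening from root to tip so the feather ends flex most.
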PhysicsAsset /BRCosmetics/Weapons/FORT_Melee/Pickaxe_Virtuous/Meshes/Virtuous_Axe_Physics
bodies {
 Melee_TwoHanded_Handle [{shapeType=Capsule, length=12.2907, radius=5.388, center=<0.5758 -19.2935 90.6956>, rotation=<rot 0 0 -90.0004>}]
 dyn_feathers_l_1 [{shapeType=Sphere, radius=2.5, shapeName=dyn_feathers_l_1_sphere}]
 dyn_feathers_l_2 [{shapeType=Sphere, radius=3, shapeName=dyn_feathers_l_2_sphere}]
 dyn_feathers_l_3 [{shapeType=Capsule, length=4, radius=3, shapeName=dyn_feathers_l_3_capsule, center=<1.3 0 -2.3>, rotation=<rot 0 0 89.9998>}]
 dyn_feathers_l_4 [{shapeType=Capsule, length=4.25, radius=2.75, shapeName=dyn_feathers_l_4_capsule, center=<1.3 0 0>, rotation=<rot 0 0 -90.0001>}]
 dyn_feathers_r_1 [{shapeType=Sphere, radius=2.5, shapeName=dyn_feathers_r_1_sphere}]
 dyn_feathers_r_2 [{shapeType=Sphere, radius=3, shapeName=dyn_feathers_r_2_sphere}]
 dyn_feathers_r_3 [{shapeType=Capsule, length=4, radius=3, shapeName=dyn_feathers_r_3_capsule, center=<-0.75 0 2.3>, rotation=<rot 0 0 89.9998>}]
 dyn_feathers_r_4 [{shapeType=Capsule, length=4.25, radius=2.75, shapeName=dyn_feathers_r_4_capsule, center=<-1.3 0 0>, rotation=<rot 0 0 -90.0001>}]
}
constraints [
 {bone1Name=dyn_feathers_r_2, bone2Name=dyn_feathers_r_1, swing1LimitDegrees=10, swing2LimitDegrees=10, twistLimitDegrees=5}
 {bone1Name=dyn_feathers_l_2, bone2Name=dyn_feathers_l_1, swing1LimitDegrees=10, swing2LimitDegrees=10, twistLimitDegrees=5}
 {bone1Name=dyn_feathers_l_1, bone2Name=Melee_TwoHanded_Handle, swing1LimitDegrees=5, swing2LimitDegrees=5, twistLimitDegrees=5}
 {bone1Name=dyn_feathers_l_3, bone2Name=dyn_feathers_l_2, swing1LimitDegrees=20, swing2LimitDegrees=15, twistLimitDegrees=10}
 {bone1Name=dyn_feathers_l_4, bone2Name=dyn_feathers_l_3, swing1LimitDegrees=20, swing2LimitDegrees=20, twistLimitDegrees=10}
 {bone1Name=dyn_feathers_r_1, bone2Name=Melee_TwoHanded_Handle, swing1LimitDegrees=5, swing2LimitDegrees=5, twistLimitDegrees=5}
 {bone1Name=dyn_feathers_r_3, bone2Name=dyn_feathers_r_2, swing1LimitDegrees=20, swing2LimitDegrees=15, twistLimitDegrees=10}
 {bone1Name=dyn_feathers_r_4, bone2Name=dyn_feathers_r_3, swing1LimitDegrees=20, swing2LimitDegrees=20, twistLimitDegrees=10}
]
